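# Physics asset for the Teriyaki Fish pirate hat face accessory, giving its dangling rag, sideburns and tricorn corners bodies for secondary motion.
# A head sphere anchors capsule bodies on the dyn_ bones.
# The rag knot joins the head with free swing and twist, and the rag's top and bottom chain off the knot with angular limits.
# Sideburns and hat corners constrain directly to the head with narrower limits, tightest on the corners.
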PhysicsAsset /BRCosmetics/Accessories/Hats/M_MED_Teriyaki_Fish_Pirate_Hat/Meshes/M_MED_Teriyaki_Fish_Pirate_FaceAcc_Physics
bodies {
 dyn_ragBottom [{shapeType=Capsule, length=5.8716, radius=2.2547, center=<3.6223 0.4395 -0.0851>, rotation=<rot 0 96.4741 -84.9148>}]
 dyn_ragKnot [{shapeType=Capsule, length=3.2633, radius=2.7372, center=<2.3793 -0.1978 0.0718>, rotation=<rot 38.6854 -90 28.5296>}]
 dyn_ragTop [{shapeType=Capsule, length=6.6053, radius=2.1778, center=<3.6217 0.0945 0.3847>, rotation=<rot -1.4825 -94.4854 -95.9367>}]
 dyn_sideburn_l [{shapeType=Capsule, length=4.7118, radius=2.5045, center=<-2.8659 0.0001 0>, rotation=<rot 90 0 0>}]
 dyn_sideburn_r [{shapeType=Capsule, length=4.7118, radius=2.5045, center=<2.9881 0 0.0003>, rotation=<rot 90 0 0>}]
 dyn_triCorner_l [{shapeType=Capsule, length=4.7918, radius=1.5901, center=<2.8872 0 0>, rotation=<rot 90 0 0>}]
 dyn_triCorner_r [{shapeType=Capsule, length=4.8063, radius=1.8019, center=<3.0071 0.0001 0>, rotation=<rot 90 0 0>}]
 head [{shapeType=Sphere, radius=12.6902}]
}
constraints [
 {
  bone1Name=dyn_ragKnot
  bone2Name=head
  swing1LimitDegrees=45
  swing1Motion=Free
  swing2LimitDegrees=45
  swing2Motion=Free
  twistLimitDegrees=45
  twistMotion=Free
 }
 {bone1Name=dyn_ragTop, bone2Name=dyn_ragKnot, swing1LimitDegrees=50, swing2LimitDegrees=20, twistLimitDegrees=15}
 {bone1Name=dyn_ragBottom, bone2Name=dyn_ragKnot, swing1LimitDegrees=50, swing2LimitDegrees=30, twistLimitDegrees=15}
 {bone1Name=dyn_sideburn_r, bone2Name=head, swing1LimitDegrees=35, swing2LimitDegrees=5, twistLimitDegrees=10}
 {bone1Name=dyn_sideburn_l, bone2Name=head, swing1LimitDegrees=35, swing2LimitDegrees=5, twistLimitDegrees=10}
 {bone1Name=dyn_triCorner_l, bone2Name=head, swing1LimitDegrees=15, swing2LimitDegrees=5, twistLimitDegrees=10}
 {bone1Name=dyn_triCorner_r, bone2Name=head, swing1LimitDegrees=15, swing2LimitDegrees=5, twistLimitDegrees=10}
]
