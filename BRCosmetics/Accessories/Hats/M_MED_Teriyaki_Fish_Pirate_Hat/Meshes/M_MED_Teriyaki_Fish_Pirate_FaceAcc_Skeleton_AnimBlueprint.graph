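# Animation Blueprint for the Teriyaki Fish Pirate face accessory skeleton, adding secondary motion on top of the pose copied from its attached parent mesh.
# Weighted multi-bone constraints place the left and right dyn_vest bones relative to spine, arm, clavicle and pelvis bones.
# AnimDynamics then springs both vest bones with planar limits driven by spine_03.
# A final RigidBody node simulates the accessory's physics asset in root-bone space; the event graph has no logic.
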AnimBlueprint /BRCosmetics/Accessories/Hats/M_MED_Teriyaki_Fish_Pirate_Hat/Meshes/M_MED_Teriyaki_Fish_Pirate_FaceAcc_Skeleton_AnimBlueprint
target_skeleton=&/M_MED_Teriyaki_Fish_Pirate_FaceAcc_Skeleton
graphs [
 AnimGraph {nodes [
   n1 = "Constraint - dyn_vest_r (PPPPPPPP)"!(
    ComponentPose=Local To Component(LocalPose=Copy Pose From Mesh() {bUseAttachedParent=true})
    Alpha=1
    ConstraintWeights_0=0.0296
    ConstraintWeights_1:float=0.0095
    ConstraintWeights_2:float=0.0409
    ConstraintWeights_3:float=0.1814
    ConstraintWeights_4:float=0.2872
    ConstraintWeights_5:float=0.1305
    ConstraintWeights_6:float=0.2966
    ConstraintWeights_7:float=0.0243
   ) -> (Pose) {
    alphaBoolBlend {blendOption=Linear}
    boneToModify {boneName=dyn_vest_r}
    constraintSetup [
     {targetBone {boneName=upperarm_r}}
     {targetBone {boneName=upperarm_twist_01_r}}
     {targetBone {boneName=spine_01}}
     {targetBone {boneName=spine_02}}
     {targetBone {boneName=spine_03}}
     {targetBone {boneName=spine_04}}
     {targetBone {boneName=clavicle_r}}
     {targetBone {boneName=pelvis}}
    ]
    constraintWeights [1,1,1,1,1,1,1,1]
   }
   n2 = "Constraint - dyn_vest_l (PPPPPPP)"!(
    ComponentPose=@n1
    Alpha=1
    ConstraintWeights_0=0.053
    ConstraintWeights_1:float=0.0288
    ConstraintWeights_2:float=0.1809
    ConstraintWeights_3:float=0.3163
    ConstraintWeights_4:float=0.0753
    ConstraintWeights_5:float=0.3331
    ConstraintWeights_6:float=0.0124
   ) -> (Pose) {
    alphaBoolBlend {blendOption=Linear}
    boneToModify {boneName=dyn_vest_l}
    constraintSetup [
     {targetBone {boneName=upperarm_twist_01_l}}
     {targetBone {boneName=spine_01}}
     {targetBone {boneName=spine_02}}
     {targetBone {boneName=spine_03}}
     {targetBone {boneName=spine_04}}
     {targetBone {boneName=clavicle_l}}
     {targetBone {boneName=pelvis}}
    ]
    constraintWeights [1,1,1,1,1,1,1]
   }
   n3 = Anim Dynamics - Bone: dyn_vest_r(ComponentPose=@n2, Alpha=1, ExternalForce="0,0,0") {
    angularDampingOverride=1
    angularSpringConstant=200
    bAngularSpring=true
    bGravityOverrideInSimSpace=true
    bOverrideAngularDamping=true
    bUseSphericalLimits=true
    simulationSpace=BoneRelative
    alphaBoolBlend {blendOption=Linear}
    boundBone {boneName=dyn_vest_r}
    componentAppliedLinearAccClamp=<500 500 500>
    componentLinearAccScale=<0.5 0.5 0.5>
    componentLinearVelScale=<0.25 0.25 0.25>
    physicsBodyDefinitions [{
      collisionType=CustomSphere
      sphereCollisionRadius=2
      boundBone {boneName=dyn_vest_r}
      boxExtents=<24 4 24>
      constraintSetup {angularLimitsMax=<-4 10 20>, angularLimitsMin=<-4 -10 -5>, angularTarget=<1 0 0>}
      localJointOffset=<4 0 0>
     }]
    planarLimits [{drivingBone {boneName=spine_03}, planeTransform {location=<0 0 16>, rotation=<rot 0 0 30>}}]
    relativeSpaceBone {boneName=root}
   }
   n4 = Anim Dynamics - Bone: dyn_vest_l(ComponentPose=@n3, Alpha=1, ExternalForce="0,0,0") {
    angularDampingOverride=1
    angularSpringConstant=200
    bAngularSpring=true
    bGravityOverrideInSimSpace=true
    bOverrideAngularDamping=true
    bUseSphericalLimits=true
    simulationSpace=BoneRelative
    alphaBoolBlend {blendOption=Linear}
    boundBone {boneName=dyn_vest_l}
    componentAppliedLinearAccClamp=<500 500 500>
    componentLinearAccScale=<0.5 0.5 0.5>
    componentLinearVelScale=<0.25 0.25 0.25>
    physicsBodyDefinitions [{
      collisionType=CustomSphere
      sphereCollisionRadius=2
      boundBone {boneName=dyn_vest_l}
      boxExtents=<24 4 24>
      constraintSetup {angularLimitsMax=<-4 10 20>, angularLimitsMin=<-4 -10 -5>, angularTarget=<1 0 0>}
      localJointOffset=<4 0 0>
     }]
    planarLimits [{drivingBone {boneName=spine_03}, planeTransform {location=<0 0 -15>, rotation=<rot 0 0 150>}}]
    relativeSpaceBone {boneName=root}
   }
   n5 = RigidBody(ComponentPose=@n4) {
    bClampLinearTranslationLimitToRefPose=true
    bOverrideWorldGravity=true
    overridePhysicsAsset=&/M_MED_Teriyaki_Fish_Pirate_FaceAcc_Physics
    simulationSpace=BaseBoneSpace
    alphaBoolBlend {blendOption=Linear}
    baseBoneRef {boneName=root}
    componentAppliedLinearAccClamp=<500 500 500>
    componentLinearAccScale=<0.6 0.6 0.6>
    componentLinearVelScale=<0.4 0.4 0.4>
    overrideWorldGravity {z=-980}
   }
   Output Pose(Result=Component To Local(ComponentPose=@n5)) {layerGroup=None}
  ]}
 EventGraph {nodes [Event BlueprintUpdateAnimation()]}
]
variables []
